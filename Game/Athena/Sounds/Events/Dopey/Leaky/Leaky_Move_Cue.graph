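# Sound cue for Leaky's movement in the Dopey event, in the scripted events sound class.
# A mixer blends close and distant attenuation branches using Sleepy's attenuation assets, though both have empty inputs and no wave is referenced.
# The mix feeds a Modulator, then a continuous modulator that maps the Progress parameter to pitch with an inverted output range.
# A concurrency override caps instances at four, stopping the oldest.
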SoundCue /Game/Athena/Sounds/Events/Dopey/Leaky/Leaky_Move_Cue
root=ModulatorContinuous_0
nodes [
 {id=Attenuation_0, children [null], properties {attenuationSettings=/Game/Athena/Sounds/Events/Dopey/Sleepy/Sleepy_MLTI_Close_Att}}
 {id=Attenuation_1, children [null], properties {attenuationSettings=/Game/Athena/Sounds/Events/Dopey/Sleepy/Sleepy_MLTI_Distant_Att}}
 {id=Mixer_0, children [Attenuation_0, Attenuation_1], properties {inputVolume [1,1]}}
 {
  id=ModulatorContinuous_0
  children [Modulator_0]
  properties {pitchModulationParams {default=0, maxOutput=0.8, minOutput=1, parameterName=Progress}}
 }
 {id=Modulator_0, children [Mixer_0]}
]
properties {
 bOverrideConcurrency=true
 maxDistance=35000
 soundClassObject=/Game/Sounds/FortSoundClassesAndModes/Fort_Events_Scripted
 volumeMultiplier=1
 attenuationOverrides {falloffDistance=3000, attenuationShapeExtents {x=0}}
 concurrencyOverrides {maxCount=4, resolutionRule=StopOldest}
}
